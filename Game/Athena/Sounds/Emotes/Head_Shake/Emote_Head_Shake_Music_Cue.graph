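# Sound cue that plays the looping music for the Head Shake emote in Athena.
# A single looping Wave Player feeds the output directly with the Emote_Head_Shake_Music_Loop wave.
# It uses the first-person emote music sound class and the matching Athena emote-music concurrency rule.
# It also has a send to the Fortnite_AudioAnalysis_Music submix.
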SoundCue /Game/Athena/Sounds/Emotes/Head_Shake/Emote_Head_Shake_Music_Cue
root=WavePlayer_0
nodes [{id=WavePlayer_0, properties {bLooping=true, soundWaveAssetPtr=&/Emote_Head_Shake_Music_Loop}}]
properties {
 maxDistance=2097152
 soundClassObject=/Game/Sounds/FortSoundClassesAndModes/Fort_Music_Emote_1P_PSM
 volumeMultiplier=1
 concurrencySet [/Game/Sounds/Concurrency_Rules/Athena_EmoteMusic_1P_Conc]
 soundSubmixSends [{minSendLevel=0, soundSubmix=&Fortnite_AudioAnalysis_Music}]
}
